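# BREnv_StaticMesh_Nanite is an Editor Utility Blueprint kept in the never-cooked BR_Env reports folder of the GART editor utilities.
# Its name points to a Nanite report for static meshes, but it holds no logic.
# The event graph is empty and no variables are declared.
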EditorUtilityBlueprint /Game/NeverCook/EditorUtilities/GART/BR_Env/Reports/BREnv_StaticMesh_Nanite
graphs [EventGraph {links [], nodes []}]
variables []
