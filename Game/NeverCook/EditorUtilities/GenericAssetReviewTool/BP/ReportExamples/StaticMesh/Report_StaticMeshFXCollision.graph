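# Editor utility report for the Generic Asset Review Tool that flags static meshes used only in Niagara FX that still have collision.
# The filter entry delegates to NoLoadLogic, which checks usage and collision from asset data, apparently without loading the mesh.
# Complex and simple collision add nested errors and raise severity, with complex ranked higher.
# It returns "FX With Collision" report data, valid only if severity stays at the base level.
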EditorUtilityBlueprint /Game/NeverCook/EditorUtilities/GenericAssetReviewTool/BP/ReportExamples/StaticMesh/Report_StaticMeshFXCollision
graphs [
 EventGraph {links [], nodes []}
 NoLoadLogic {nodes [
   n1 = NoLoadLogic!() -> (then, AssetData:Asset Data Structure) {is=title K2Node_FunctionEntry}
   > Set l_asset(l_asset:Asset Data Structure=@n1) -> (Output_Get:Asset Data Structure)
   > n2 = Sequence()
   > n3 = IsMeshOnlyUsedInFX!(execute, assetData:Asset Data Structure=Get l_asset() -> (l_asset:Asset Data Structure), __WorldContext:Object Reference=Get l_world()) -> (then, OnlyFX:bool)
   > Branch(Condition=@n3)
   > n4 = NoLoad_GetStaticMeshCollisionData!(execute, AssetData:Asset Data Structure=Get l_asset(), __WorldContext:Object Reference=Get l_world()) -> (then, SectionsWithCollision:int=0, CollisionPrimCount:int=0, CollisionComplexity:String)
   n5 = "integer > integer"(A=@n4.CollisionPrimCount)
   n6 = "integer > integer"(A=@n4)
   > Branch(Condition=OR Boolean(A=@n6, B=@n5))
   > n7 = Sequence!(execute) -> (then_0, then_1, then_2:Exec)
   > AddError(NewError="Has Collision and is only referenced by Niagara.  Niagara Meshes do not need collision so usually this is bad.")
   Branch(execute=@n7.then_1, Condition=@n6)
   > AddError(NewError=Has Complex Collision which is expensive memory wise., Depth=1)
   > SetSeverity!(execute, self:GART Report Base Object Reference, NewSeverity:Enum_GART_ReportSeverity Enum=NewEnumerator2, OnlyIfHigher:bool=true) -> (then)
   Branch(execute=@n7.then_2, Condition=@n5)
   > AddError(NewError=Has Simple Collision which is less of a problem but not needed., Depth=1)
   > SetSeverity(NewSeverity=NewEnumerator1, OnlyIfHigher=true)
   Return Node!(
    execute=@n2.then_1
    Valid:bool="Equal (Enum)"(A:Enum_GART_ReportSeverity Enum=Get l_severity(), B:Enum_GART_ReportSeverity Enum=NewEnumerator0)
    ReportData:Struct GART Report Data Structure=Make Struct GART Report Data(
     DisplayTitle_20_#1=FX
     DisplayDescShort_29_#2=FX With Collision
     Severity_46_#3=Get l_severity()
    )
   )
  ]}
 RunFilterLogicOnAssetData {nodes [
   n1 = RunFilterLogicOnAssetData()
   > n2 = NoLoadLogic!(execute, self, AssetData:Asset Data Structure=@n1) -> (then, Valid:bool, ReportData:Struct GART Report Data Structure) {is=title K2Node_CallFunction}
   > Return Node(Valid=@n2, ReportData=@n2.ReportData)
  ]}
]
variables []
