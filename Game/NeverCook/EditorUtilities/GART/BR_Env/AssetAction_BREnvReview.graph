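# Editor asset action "Environment Review Assets" that checks selected environment assets, mainly static meshes and textures, against BR environment rules.
# It first drops assets whose paths match the ignore list, such as SourceMeshes.
# By default it runs the ContentReviewToolBREnv review classes and displays their results.
# With Ctrl held, it instead runs GART report filters with loading allowed and opens the results in the EUW_GART_List_Main tab.
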EditorUtilityBlueprint /Game/NeverCook/EditorUtilities/GART/BR_Env/AssetAction_BREnvReview
graphs [
 Environment Review Assets {nodes [
   Environment Review Assets() {is=title K2Node_FunctionEntry}
   > n1 = GetSelectedAssetData()
   > n2 = GetEditorWorld(self=Get UnrealEditorSubsystem())
   > n3 = Sequence()
   > n4 = For Each Loop(Array:Array of Asset Data Structures=@n1) -> (Array Element:Asset Data Structure)
   > n5 = CheckPathIgnoreList!(
    execute
    AssetData:Asset Data Structure=@n4
    IgnoreList:Array of Strings=Get IgnoreList() -> (IgnoreList:Array of Strings)
    __WorldContext:Object Reference=@n2
   ) -> (then, IgnoreAsset:bool)
   > n6 = Branch(Condition=@n5)
   Add(
    execute=@n6.else
    TargetArray:Array of Asset Data Structures=Get l_AllowedAssetDatas() -> (l_AllowedAssetDatas:Array of Asset Data Structures)
    NewItem:"Asset Data Structure (by ref)"=@n4
   ) {is=title K2Node_CallArrayFunction}
   n7 = Branch(
    execute=@n3.then_1
    Condition=Is Control Down!(KeysState:"Modifier Keys State Structure (by ref)"=GetModifierKeysState!() -> (ReturnValue:Modifier Keys State Structure)) -> (ReturnValue:bool)
   )
   > n8 = RunAssetsThroughFilters!(
    execute
    AssetDatas:Array of Asset Data Structures=Get l_AllowedAssetDatas()
    ReportClasses:Array of GART Report Base Soft Class References=Get ReportClasses() -> (ReportClasses:Array of GART Report Base Soft Class References)
    AllowLoading:bool=true
    __WorldContext:Object Reference=@n2
   ) -> (then, ReportDataMap:Map of Names to Struct GART Asset Report List Structures)
   > n9 = SpawnAndRegisterTabAndGetID!(
    execute
    self:Editor Utility Subsystem Object Reference=Get EditorUtilitySubsystem()
    InBlueprint:Editor Utility Widget Blueprint Object Reference=/Game/NeverCook/EditorUtilities/GenericAssetReviewTool/Widgets/EUW_GART_List_Main
   ) -> (then, NewTabID:Name=None, ReturnValue:Editor Utility Widget Object Reference)
   > n10 = Cast To EUW_GART_List_Main!(execute, Object=@n9.ReturnValue) -> (then, CastFailed, AsEUW GART List Main:EUW GART List Main Object Reference) {is=title K2Node_DynamicCast}
   > SetAssetReportData!(
    execute
    self:EUW GART List Main Object Reference=@n10
    ReportData:Map of Names to Struct GART Asset Report List Structures=@n8
    UpdateList:bool=true
   ) -> (then)
   n11 = Self-Reference()
   n12 = Run Reviews on Assets!(
    execute=@n7.else
    Assets:Array of Asset Data Structures=Get l_AllowedAssetDatas()
    Reviews:Array of Soft Class References=Get ReviewsToRun() -> (ReviewsToRun:Array of Object Soft Class References)
    __WorldContext:Object Reference=@n11
   ) -> (then, AssetReviewMap:Map of Asset Data Structures to Structures)
   > Display Review Results!(
    execute
    AssetMap:Map of Asset Data Structures to Structures=@n12
    ObjectMap:Map of Object References to Structures=Get l_objects() -> (l_objects:Map of Object References to Structures)
    __WorldContext:Object Reference=@n11
   ) -> (then)
  ]}
 EventGraph {nodes [Event Run()]}
 IsActionForBlueprints {nodes [
   IsActionForBlueprints() {is=title K2Node_FunctionEntry}
   > Return Node!(execute, ReturnValue:bool=Parent: IsActionForBlueprints!() -> (ReturnValue:bool) {is=title K2Node_CallParentFunction})
  ]}
]
variables [
 IgnoreList {type="array<string>", default [SourceMeshes]}
 ReportClasses {
  type="array<Class>"
  default [
   /Game/NeverCook/EditorUtilities/GART/BR_Env/Reports/BREnv_BP_ClassMeshVertCount.BREnv_BP_ClassMeshVertCount_C
   /Game/NeverCook/EditorUtilities/GART/BR_Env/Reports/BREnv_StaticMesh_CollisionLOD.BREnv_StaticMesh_CollisionLOD_C
   /Game/NeverCook/EditorUtilities/GART/BR_Env/Reports/BREnv_StaticMesh_LODDif.BREnv_StaticMesh_LODDif_C
   /Game/NeverCook/EditorUtilities/GART/BR_Env/Reports/BREnv_StaticMesh_Nanite.BREnv_StaticMesh_Nanite_C
   /Game/NeverCook/EditorUtilities/GART/BR_Env/Reports/BREnv_Texture_EmptyAlpha.BREnv_Texture_EmptyAlpha_C
   /Game/NeverCook/EditorUtilities/GenericAssetReviewTool/BP/ReportExamples/StaticMesh/Report_StaticMeshFXCollision.Report_StaticMeshFXCollision_C
  ]
 }
 ReviewsToRun {
  type="array<Class>"
  default [
   /ContentReviewToolBREnv/Reviews/StaticMesh/Review_SM_LODForCollision_BR.Review_SM_LODForCollision_BR_C
   /ContentReviewToolBREnv/Reviews/StaticMesh/Review_SM_MinLOD_BR.Review_SM_MinLOD_BR_C
   /ContentReviewToolBREnv/Reviews/StaticMesh/Review_SM_NaniteSettings_BR.Review_SM_NaniteSettings_BR_C
   /ContentReviewToolBREnv/Reviews/StaticMesh/Review_SM_ClassVertexCount_BR.Review_SM_ClassVertexCount_BR_C
   /ContentReviewToolBREnv/Reviews/StaticMesh/Review_SM_LODDif_BR.Review_SM_LODDif_BR_C
   /ContentReviewToolBREnv/Reviews/Texture/Review_Tex_MemorySize_BR.Review_Tex_MemorySize_BR_C
   /ContentReviewToolBREnv/Reviews/Texture/Review_Tex_PowerOf2_BR.Review_Tex_PowerOf2_BR_C
   /ContentReviewToolBREnv/Reviews/Texture/Review_Tex_EmptyAlpha_BR.Review_Tex_EmptyAlpha_BR_C
   /ContentReviewToolBREnv/Reviews/Texture/Review_Tex_EmptyChannels_BR.Review_Tex_EmptyChannels_BR_C
   /ContentReviewToolBREnv/Reviews/StaticMesh/Review_SM_ShotgridID.Review_SM_ShotgridID_C
   /ContentReviewToolBREnv/Reviews/StaticMesh/Review_SM_BlueprintRef.Review_SM_BlueprintRef_C
  ]
 }
]
