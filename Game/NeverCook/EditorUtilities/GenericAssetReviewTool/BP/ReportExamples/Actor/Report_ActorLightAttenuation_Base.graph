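# An editor-only GART report Blueprint that flags lights with too large an attenuation radius for their detail mode.
# RunComponentLogic passes early if the parent check succeeds, otherwise reads the point light's radius and looks up a recommended/maximum budget in DetailModeRadiusBudgets.
# Missing modes attempt a fallback to a higher detail mode's budget.
# Exceeding the lower limit logs a formatted error, and exceeding the upper changes the returned severity.
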EditorUtilityBlueprint /Game/NeverCook/EditorUtilities/GenericAssetReviewTool/BP/ReportExamples/Actor/Report_ActorLightAttenuation_Base
graphs [
 EventGraph {links [], nodes []}
 RunComponentLogic {nodes [
   n1 = RunComponentLogic!() -> (then, Comp:Light Component Object Reference) {is=title K2Node_FunctionEntry}
   > n2 = Set l_comp(l_comp:Light Component Object Reference=@n1) -> (Output_Get:Light Component Object Reference)
   > n3 = Parent: RunComponentLogic!(execute, Comp:Light Component Object Reference=@n2) -> (then, Success?:bool, Error:Array of Strings, severity:Enum_GART_ReportSeverity Enum) {is=title K2Node_CallParentFunction}
   > n4 = Branch(Condition=@n3)
   > Return Node!(execute, Success?:bool=true, Error:Array of Strings, severity:Enum_GART_ReportSeverity Enum=NewEnumerator0)
   n5 = Sequence!(execute=@n4.else) -> (then_0, then_1, then_2:Exec, then_3:Exec)
   > n6 = Cast To PointLightComponent!(execute, Object=Get l_comp() -> (l_comp:Light Component Object Reference)) -> (then, CastFailed, AsPoint Light Component:Point Light Component Object Reference) {is=title K2Node_DynamicCast}
   > Set l_radius(l_radius:float=Get AttenuationRadius!(self:Local Light Component Object Reference=@n6) -> (AttenuationRadius:float)) -> (Output_Get:float)
   > Set l_castShadows(l_castShadows:bool=AND Boolean(
     A=Get CastShadows!(self:Light Component Base Object Reference=@n6) -> (CastShadows:bool)
     B=Get CastDynamicShadows!(self:Light Component Base Object Reference=@n6) -> (CastDynamicShadows:bool)
    )) -> (Output_Get:bool)
   n7 = Find(
    TargetMap:Map of EDetailMode Enums to Vector 2D Structures=Get DetailModeRadiusBudgets() -> (DetailModeRadiusBudgets:Map of EDetailMode Enums to Vector 2D Structures=DM_Low)
    Key:"EDetailMode Enum (by ref)"=Get DetailMode!(self:Scene Component Object Reference=Get l_comp()) -> (DetailMode:EDetailMode Enum=DM_Low)
   ) -> (Value:Vector 2D Structure)
   n8 = Branch(execute=@n5.then_1, Condition=@n7.ReturnValue)
   > Set l_budget(l_budget:Vector 2D Structure=@n7) -> (Output_Get:Vector 2D Structure)
   n9 = Keys!(
    execute=@n8.else
    TargetMap:Map of EDetailMode Enums to Vector 2D Structures=Get DetailModeRadiusBudgets() -> (DetailModeRadiusBudgets=DM_Low)
   ) -> (then, Keys:Array of EDetailMode Enums)
   > n10 = For Each Loop(Array:Array of EDetailMode Enums=@n9) -> (Array Element:EDetailMode Enum)
   Branch(execute=@n10.Completed, Condition="integer < integer"(A=Get l_index() -> (l_index:int=0), B=5))
   n11 = Find(
    TargetMap=Get DetailModeRadiusBudgets() -> (DetailModeRadiusBudgets=DM_Low)
    Key=Get l_lastFoundDetailMode() -> (l_lastFoundDetailMode:EDetailMode Enum=DM_Low)
   )
   > Branch(Condition=@n11.ReturnValue)
   > Set l_budget(l_budget=@n11)
   n12 = Get DetailMode(self=Get l_comp()) -> (DetailMode=DM_Low)
   n13 = "To Integer (Byte)"!(InByte:Byte=@n10) -> (ReturnValue:int=0)
   Branch(execute=@n10.LoopBody, Condition="integer > integer"(A=@n13, B="To Integer (Byte)"(InByte=@n12) -> (ReturnValue=0)))
   > Set l_index(l_index:int="Min (Integer)"(A=@n13, B=Get l_index() -> (l_index=0)) -> (ReturnValue=0)) -> (Output_Get:int=0)
   > Set l_lastFoundDetailMode(l_lastFoundDetailMode:EDetailMode Enum=@n10) -> (Output_Get:EDetailMode Enum=DM_Low)
   n14 = BreakVector2D(InVec=Get l_budget() -> (l_budget:Vector 2D Structure))
   Branch(execute=@n5.then_2, Condition=AND Boolean(A="Not Equal (Float)"(A=@n14), B="Not Equal (Float)"(A=@n14.Y)))
   n15 = BreakVector2D(InVec=Get l_budget())
   > Branch(Condition="float > float"(
     A=Get l_radius() -> (l_radius:float)
     B="Min (Float)"(A:double=@n15, B:double=@n15.Y) -> (ReturnValue:double) {is=title K2Node_CommutativeAssociativeBinaryOperator}
    ))
   n16 = BreakVector2D(InVec=Get l_budget())
   n17 = "Min (Float)"(A=@n16, B=@n16.Y)
   n18 = "Max (Float)"(A=@n16, B=@n16.Y)
   n19 = Enum to String!(Enumerator:EDetailMode Enum=Get DetailMode(self=Get l_comp()) -> (DetailMode=DM_Low)) -> (ReturnValue:String)
   n20 = Get l_budget()
   n21 = Select(
    Option 0:Text="INVTEXT(\"Attenuation Radius is very large.\")"
    Option 1:Text="INVTEXT(\"Attenuation Radius is too large!\")"
    Index="float > float"(A=Get l_radius(), B=BreakVector2D(InVec=@n20).Y)
   ) -> (ReturnValue:Text)
   n22 = Get l_radius()
   > Add(
    TargetArray=Get l_errors() -> (l_errors:Array of Strings)
    NewItem="To String (Text)"(InText=Format Text!(
      Format="INVTEXT(\"{ErrorBase}  It is {radius} and lights using {DetailMode} detail mode MUST be less than {max} but is recommended to be less than {min}\")"
      ErrorBase:Text=@n21
      radius:float=@n22
      DetailMode:String=@n19
      max:double=@n18
      min:double=@n17
     ) -> (Result))
   ) {is=title K2Node_CallArrayFunction}
   n23 = Get l_errors()
   n24 = BreakVector2D(InVec=Get l_budget())
   n25 = "Max (Float)"(A=@n24, B=@n24.Y)
   n26 = Get l_radius()
   Return Node(
    execute=@n5.then_3
    Success?=NOT Boolean(A=Is Valid Index!(TargetArray=@n23, IndexToTest:int) -> (ReturnValue:bool))
    Error=Get l_errors()
    severity=Select(
     Option 0:Enum_GART_ReportSeverity Enum=NewEnumerator1
     Option 1:Enum_GART_ReportSeverity Enum=NewEnumerator2
     Index="float > float"(A=@n26, B=@n25)
    ) -> (ReturnValue:Enum_GART_ReportSeverity Enum)
   )
  ]}
]
variables [DetailModeRadiusBudgets {default {DM_High=<3500 10000>, DM_Low=<500 1500>, DM_Medium=<1000 5000>}}]
