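# Editor review-tool report rule that flags floor-attach building actors with no ParentActorToAttachTo that sit on another building actor.
# RunFilterLogicOnObject screens out excluded classes, checks parent and attachment type, then reports a possible manual attach target by label.
# DoLocationCheck finds that target with a debug-drawn downward line trace, filtered to BuildingSMActors with attachment slot None.
# DoTrace, a sphere-trace alternative using the attachment radius, is present but not wired into execution.
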EditorUtilityBlueprint /Game/NeverCook/EditorUtilities/GenericAssetReviewTool/BP/ReportExamples/Actor/Report_BuildingAttachment_Base
graphs [
 DoLocationCheck {nodes [
   n1 = DoLocationCheck!() -> (then, CheckBuildingActor:Building Static Mesh Object Reference) {is=title K2Node_FunctionEntry}
   > PrintString(InString=Loc)
   n2 = Get Actor Bounds(self=@n1, bOnlyCollidingComponents=true)
   > Set l_start(l_start:Vector=@n2) -> (Output_Get:Vector)
   n3 = BreakVector(InVec=@n2)
   n4 = Get Actor Location(self=@n1)
   n5 = float * float(A=BreakVector(InVec=@n2.BoxExtent).Z, B=1.2) {is=title K2Node_PromotableOperator}
   > n6 = Multi Line Trace For Objects!(
    execute
    WorldContextObject:Object Reference=@n1
    Start:Vector=MakeVector(X=@n3, Y=@n3.Y, Z=BreakVector(InVec=@n4).Z)
    End:Vector=vector - vector(A=Get l_start() -> (l_start:Vector), B=MakeVector(Y=0, Z=@n5))
    ObjectTypes:Array of EObjectTypeQuery Enums=Make Array("[0]":EObjectTypeQuery Enum=ObjectTypeQuery1) -> (Array:Array of EObjectTypeQuery Enums)
    bTraceComplex:bool=false
    ActorsToIgnore:Array of Actor Object References=Make Array("[0]":Actor Object Reference=@n1) -> (Array:Array of Actor Object References)
    DrawDebugType:EDrawDebugTrace Enum=ForDuration
    bIgnoreSelf:bool=true
    TraceColor:Linear Color Structure=(R=1,G=0,B=0,A=1)
    TraceHitColor:Linear Color Structure=(R=0,G=1,B=0,A=1)
    DrawTime:float=5
   ) -> (then, OutHits:Array of Hit Result Structures, ReturnValue:bool)
   > PrintString(InString=Hello)
   > n7 = FilterTargetHitResultList!(
    execute
    WorldContextObject:Object Reference=@n1
    Filter:"Fort Target Filter Structure (by ref)"=Make Fort Target Filter!(
     ActorTypeFilter:EFortTargetSelectionFilter Enum=Damageable
     ActorClassFilter:Object Class Reference=/Script/FortniteGame.BuildingSMActor
     bExcludeInstigator:bool=True
     bUseTrapsOwningPawnAsInstigator:bool=False
     bExcludeRequester:bool=True
     bExcludeAllAttachedToInstigator:bool=true
     bExcludeAthenaVehicleOfInstigator:bool=False
     bExcludeAllAttachedToRequester:bool=False
     bExcludeTargetsInEditMode:bool=False
     bExcludePawnFriends:bool=True
     bExcludeFriendlyAI:bool=False
     bIncludeRiddenAI:bool=False
     bExcludeAllAI:bool=False
     bExcludePawnNeutrals:bool=False
     bExcludePawnEnemies:bool=False
     bExcludePawnEnemiesKeepNeutrals:bool=False
     bExcludeNonPawnFriends:bool=False
     bExcludeNonPawnEnemies:bool=False
     bConsiderFriendlyFireWhenExcludingFriends:bool=False
     bExcludeDBNOPawns:bool=False
     bAllowDBNOOverride:bool=False
     bExcludeFallingPawns:bool=False
     bExcludeWaterVolumes:bool=False
     bExcludeWaterBodies:bool=True
     bExcludeAthenaVehicleOccupiedBySource:bool=True
     bExcludeAthenaVehicleUnoccupied:bool=False
     bExcludeAthenaVehicleOccupied:bool=False
     bExcludeAthenaVehicleFromObstructionChecks:bool=False
     bExcludeWithoutNavigationCorridor:bool=False
     bExcludeNonPlayerBuiltPieces:bool=False
     bExcludePlayerBuiltPieces:bool=False
     bExcludeNonBGABuildings:bool=False
     bExcludeNonPhysicsSimulatedBuildingActors:bool=False
     bExcludeNonBlockingHits:bool=False
     bExcludeProjectiles:bool=False
     bTraceComplexCollision:bool=False
     bExcludeLandscape:bool=True
     bExcludeLandscapeSplineActors:bool=True
     bNewtonPhysicsPropsBypassAthenaFilter:bool=False
     bConsiderPhysicsPawnsAsNonPlayerPawns:bool=false
     bForceIncludePhysicProp:bool=False
     bIncludeNonDamageableAsBlocking:bool=False
     ExclusionGameplayTags:Gameplay Tag Container Structure=(GameplayTags=)
    ) -> (FortTargetFilter:Fort Target Filter Structure) {is=title K2Node_MakeStruct}
    RequestingActor:Actor Object Reference=@n1
    HitResultsToFilter:Array of Hit Result Structures=@n6
   ) -> (then, bAtLeastOneActorLeft:bool, ReturnValue:Array of Hit Result Structures)
   > n8 = For Each Loop(Array:Array of Hit Result Structures=@n7.ReturnValue) -> (Array Element:Hit Result Structure)
   Return Node!(execute=@n8.Completed, NearAnotherBuildingActor:bool=false, ActorFound:Building Static Mesh Object Reference)
   n9 = BreakHitResult!(Hit:"Hit Result Structure (by ref)"=@n8) -> (
    bBlockingHit:bool
    bInitialOverlap:bool
    Time:float
    Distance:float
    Location:Vector
    ImpactPoint:Vector
    Normal:Vector
    ImpactNormal:Vector
    PhysMat:Physical Material Object Reference
    HitActor:Actor Object Reference
    HitComponent:Primitive Component Object Reference
    HitBoneName:Name
    BoneName:Name
    HitItem:int
    ElementIndex:int
    FaceIndex:int
    TraceStart:Vector
    TraceEnd:Vector
   )
   > PrintString(InString=GetDisplayName(Object=@n9.HitActor))
   > n10 = Cast To BuildingSMActor!(execute, Object=@n9.HitActor) -> (then, CastFailed, AsBuilding Static Mesh:Building Static Mesh Object Reference) {is=title K2Node_DynamicCast}
   > Branch(Condition="Equal (Enum)"(
     A:EBuildingAttachmentSlot Enum=Get BuildingAttachmentSlot!(self:Building Static Mesh Object Reference=@n10) -> (BuildingAttachmentSlot:EBuildingAttachmentSlot Enum=SLOT_Floor)
     B:EBuildingAttachmentSlot Enum=SLOT_None
    ))
   > n11 = Set l_parentFound(l_parentFound:Building Static Mesh Object Reference=@n10) -> (Output_Get:Building Static Mesh Object Reference)
   > Return Node(NearAnotherBuildingActor=true, ActorFound=@n11)
  ]}
 DoTrace {nodes [
   n1 = Multi Line Trace For Objects(
    Start="0, 0, 0"
    End="0, 0, 0"
    ObjectTypes=ObjectTypeQuery1
    bTraceComplex=false
    DrawDebugType=ForDuration
    bIgnoreSelf=true
    TraceColor=(R=1,G=0,B=0,A=1)
    TraceHitColor=(R=0,G=1,B=0,A=1)
    DrawTime=5
   )
   > PrintString(InString=Hello)
   > n2 = FilterTargetHitResultList(
    Filter=Make Fort Target Filter(
     ActorTypeFilter=Damageable
     ActorClassFilter=&BuildingSMActor
     bExcludeInstigator=True
     bUseTrapsOwningPawnAsInstigator=False
     bExcludeRequester=True
     bExcludeAllAttachedToInstigator=true
     bExcludeAthenaVehicleOfInstigator=False
     bExcludeAllAttachedToRequester=False
     bExcludeTargetsInEditMode=False
     bExcludePawnFriends=True
     bExcludeFriendlyAI=False
     bIncludeRiddenAI=False
     bExcludeAllAI=False
     bExcludePawnNeutrals=False
     bExcludePawnEnemies=False
     bExcludePawnEnemiesKeepNeutrals=False
     bExcludeNonPawnFriends=False
     bExcludeNonPawnEnemies=False
     bConsiderFriendlyFireWhenExcludingFriends=False
     bExcludeDBNOPawns=False
     bAllowDBNOOverride=False
     bExcludeFallingPawns=False
     bExcludeWaterVolumes=False
     bExcludeWaterBodies=True
     bExcludeAthenaVehicleOccupiedBySource=True
     bExcludeAthenaVehicleUnoccupied=False
     bExcludeAthenaVehicleOccupied=False
     bExcludeAthenaVehicleFromObstructionChecks=False
     bExcludeWithoutNavigationCorridor=False
     bExcludeNonPlayerBuiltPieces=False
     bExcludePlayerBuiltPieces=False
     bExcludeNonBGABuildings=False
     bExcludeNonPhysicsSimulatedBuildingActors=False
     bExcludeNonBlockingHits=False
     bExcludeProjectiles=False
     bTraceComplexCollision=False
     bExcludeLandscape=True
     bExcludeLandscapeSplineActors=True
     bNewtonPhysicsPropsBypassAthenaFilter=False
     bConsiderPhysicsPawnsAsNonPlayerPawns=false
     bForceIncludePhysicProp=False
     bIncludeNonDamageableAsBlocking=False
     ExclusionGameplayTags=(GameplayTags=)
    )
    HitResultsToFilter=@n1
   )
   > n3 = For Each Loop(Array=@n2.ReturnValue)
   Return Node!(execute=@n3.Completed, NearAnotherBuildingActor=false, HitActors:Array of Building Static Mesh Object References)
   n4 = BreakHitResult(Hit=@n3)
   > PrintString(InString=GetDisplayName(Object=@n4.HitActor))
   > n5 = Cast To BuildingSMActor(Object=@n4.HitActor)
   > Branch(Condition="Equal (Enum)"(A=Get BuildingAttachmentSlot(self=@n5) -> (BuildingAttachmentSlot=SLOT_Floor), B=SLOT_None))
   > Set l_parentFound(l_parentFound=@n5)
   > Return Node(NearAnotherBuildingActor=true)
   n6 = DoTrace!() -> (then, CheckBuildingActor:Building Static Mesh Object Reference) {is=title K2Node_FunctionEntry}
   > Set l_actor(l_actor:Building Static Mesh Object Reference=@n6) -> (Output_Get:Building Static Mesh Object Reference)
   > n7 = Set l_start(l_start=Get Actor Bounds(self=@n6, bOnlyCollidingComponents=true))
   > n8 = Multi Sphere Trace For Objects!(
    execute
    WorldContextObject:Object Reference=Get l_actor() -> (l_actor:Building Static Mesh Object Reference)
    Start:Vector=@n7
    End:Vector=@n7
    Radius:float=Get BuildingAttachmentRadius!(self:Building Static Mesh Object Reference=Get l_actor()) -> (BuildingAttachmentRadius:float)
    ObjectTypes:Array of EObjectTypeQuery Enums=Make Array("[0]":EObjectTypeQuery Enum=ObjectTypeQuery1) -> (Array:Array of EObjectTypeQuery Enums)
    bTraceComplex:bool=false
    ActorsToIgnore:Array of Actor Object References=Make Array("[0]":Actor Object Reference=Get l_actor()) -> (Array:Array of Actor Object References)
    DrawDebugType:EDrawDebugTrace Enum=ForDuration
    bIgnoreSelf:bool=true
    TraceColor:Linear Color Structure=(R=1,G=0,B=0,A=1)
    TraceHitColor:Linear Color Structure=(R=0,G=1,B=0,A=1)
    DrawTime:float=5
   ) -> (then, OutHits:Array of Hit Result Structures, ReturnValue:bool)
   n9 = Get l_actor()
   > n10 = FilterTargetHitResultList(
    WorldContextObject=@n9
    Filter=Make Fort Target Filter(
     ActorTypeFilter=Damageable
     ActorClassFilter=&BuildingSMActor
     bExcludeInstigator=True
     bUseTrapsOwningPawnAsInstigator=False
     bExcludeRequester=True
     bExcludeAllAttachedToInstigator=true
     bExcludeAthenaVehicleOfInstigator=False
     bExcludeAllAttachedToRequester=False
     bExcludeTargetsInEditMode=False
     bExcludePawnFriends=True
     bExcludeFriendlyAI=False
     bIncludeRiddenAI=False
     bExcludeAllAI=False
     bExcludePawnNeutrals=False
     bExcludePawnEnemies=False
     bExcludePawnEnemiesKeepNeutrals=False
     bExcludeNonPawnFriends=False
     bExcludeNonPawnEnemies=False
     bConsiderFriendlyFireWhenExcludingFriends=False
     bExcludeDBNOPawns=False
     bAllowDBNOOverride=False
     bExcludeFallingPawns=False
     bExcludeWaterVolumes=False
     bExcludeWaterBodies=True
     bExcludeAthenaVehicleOccupiedBySource=True
     bExcludeAthenaVehicleUnoccupied=False
     bExcludeAthenaVehicleOccupied=False
     bExcludeAthenaVehicleFromObstructionChecks=False
     bExcludeWithoutNavigationCorridor=False
     bExcludeNonPlayerBuiltPieces=False
     bExcludePlayerBuiltPieces=False
     bExcludeNonBGABuildings=False
     bExcludeNonPhysicsSimulatedBuildingActors=False
     bExcludeNonBlockingHits=False
     bExcludeProjectiles=False
     bTraceComplexCollision=False
     bExcludeLandscape=True
     bExcludeLandscapeSplineActors=True
     bNewtonPhysicsPropsBypassAthenaFilter=False
     bConsiderPhysicsPawnsAsNonPlayerPawns=false
     bForceIncludePhysicProp=False
     bIncludeNonDamageableAsBlocking=False
     ExclusionGameplayTags=(GameplayTags=)
    )
    RequestingActor=@n9
    HitResultsToFilter=@n8
   )
   > n11 = For Each Loop(Array=@n10.ReturnValue)
   n12 = Get l_hitActors() -> (l_hitActors:Array of Building Static Mesh Object References)
   Return Node(
    execute=@n11.Completed
    NearAnotherBuildingActor=Is Valid Index!(TargetArray:Array of Building Static Mesh Object References=@n12, IndexToTest:int) -> (ReturnValue:bool)
    HitActors=@n12
   )
   > n13 = Cast To BuildingSMActor(Object=BreakHitResult(Hit=@n11).HitActor)
   > Add(TargetArray:Array of Building Static Mesh Object References=Get l_hitActors(), NewItem:"Building Static Mesh Object Reference (by ref)"=@n13) {is=title K2Node_CallArrayFunction}
  ]}
 EventGraph {links [], nodes []}
 FilterActor {nodes [
   n1 = FilterActor!() -> (then, Object:Object Reference) {is=title K2Node_FunctionEntry}
   > n2 = To Array!(execute, A:Set of Object Soft Class References=Get FilterClasses() -> (FilterClasses:Set of Object Soft Class References)) -> (then, Result:Array of Object Soft Class References)
   > n3 = For Each Loop(Array:Array of Object Soft Class References=@n2) -> (Array Element:Object Soft Class Reference)
   Return Node!(execute=@n3.Completed, Valid:bool)
   > "IsA ( soft )"!(execute, Object:Object Reference=@n1, SoftClass:Object Soft Class Reference=@n3) -> (True:Exec, False:Exec)
   > Return Node(Valid=false)
  ]}
 RunFilterLogicOnObject {nodes [
   n1 = RunFilterLogicOnObject!() -> (then, object:Object Reference) {is=title K2Node_FunctionEntry}
   > n2 = Sequence()
   > n3 = Cast To BuildingSMActor(Object=@n1)
   > Set l_smActor(l_smActor:Building Static Mesh Object Reference=@n3) -> (Output_Get:Building Static Mesh Object Reference)
   > n4 = FilterActor!(execute, self, Object:Object Reference=@n3) -> (then, Valid:bool) {is=title K2Node_CallFunction}
   > Branch(Condition=@n4)
   n5 = Get ParentActorToAttachTo!(self:Building Static Mesh Object Reference=@n3) -> (ParentActorToAttachTo:Building Static Mesh Object Reference)
   > Branch(Condition=NOT Boolean(A=IsValid(Object=@n5)))
   n6 = Get l_smActor() -> (l_smActor:Building Static Mesh Object Reference)
   > Branch(Condition="Equal (Enum)"(
     A:EBuildingAttachmentType Enum=Get BuildingAttachmentType!(self:Building Static Mesh Object Reference=@n6) -> (BuildingAttachmentType:EBuildingAttachmentType Enum=ATTACH_Floor)
     B:EBuildingAttachmentType Enum=ATTACH_Floor
    ))
   > n7 = DoLocationCheck!(execute, self, CheckBuildingActor:Building Static Mesh Object Reference=Get l_smActor()) -> (then, NearAnotherBuildingActor:bool, ActorFound:Building Static Mesh Object Reference) {is=title K2Node_CallFunction}
   > Branch(Condition=@n7)
   n8 = Format Text!(
    Format="INVTEXT(\"Building Actor has an empty ParentActorToAttachTo property and is on top of another Building Actor set to AttachSlot None.  Possibly we need to attach this by hand to {actor_label}\")"
    actor_label:String=GetActorLabel(self=@n7.ActorFound, bCreateIfNone=true)
   ) -> (Result)
   > Return Node!(
    execute
    Valid=false
    ReportData:Struct GART Report Data Structure=Make Struct GART Report Data(
     DisplayTitle_20_#1=Building Actor
     DisplayDescShort_29_#2=No Attach Target
     DisplayDescLong_30_#3="To String (Text)"(InText=@n8)
     Severity_46_#4=NewEnumerator1
    )
   )
   Return Node(execute=@n2.then_1)
   DoTrace!(execute, self, CheckBuildingActor:Building Static Mesh Object Reference) -> (then, NearAnotherBuildingActor:bool, HitActors:Array of Building Static Mesh Object References) {is=title K2Node_CallFunction}
   n9 = Get l_smActor()
   "Equal (Enum)"(A=Get BuildingAttachmentType(self=@n9) -> (BuildingAttachmentType=ATTACH_Floor), B=ATTACH_Floor)
   "Equal (Enum)"(
    A:EBuildingAttachmentSlot Enum=Get BuildingAttachmentSlot(self=@n9) -> (BuildingAttachmentSlot=SLOT_Floor)
    B:EBuildingAttachmentSlot Enum=SLOT_Floor
   )
  ]}
]
variables [FilterClasses {type="array<Class>", default [/Script/FortniteGame.BuildingServerOnlyContainer]}]
