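# Editor utility that runs environment review checks on selected or all level actors, covering lighting, attachment and damage flags.
# Holding Ctrl runs the ReviewsToRun classes and displays their results.
# Otherwise it filters actors through the GART Report Classes and feeds the report map to an EUW_GART_List_Main tab.
# Only the all-actors path checks for an empty map first; the selection path's check has no incoming execution.
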EditorUtilityBlueprint /Game/NeverCook/EditorUtilities/GART/BR_Env/BREnv_ActorActions
graphs [
 EventGraph {links [], nodes []}
 Review Selected Actors {nodes [
   Review Selected Actors() {is=title K2Node_FunctionEntry}
   > n1 = GetSelectionSet(self=Get LevelEditorSubsystem())
   > n2 = GetSelectedObjects!(execute, self:Typed Element Selection Set Object Reference=@n1, InRequiredClass:Object Class Reference) -> (then, ReturnValue:Array of Object References)
   > n3 = Branch(Condition=Is Control Down!(KeysState:"Modifier Keys State Structure (by ref)"=GetModifierKeysState!() -> (ReturnValue:Modifier Keys State Structure)) -> (ReturnValue:bool))
   n4 = Self-Reference()
   > n5 = Run Reviews on Objects!(
    execute
    Objects:Array of Object References=@n2
    Reviews:Array of Soft Class References=Get ReviewsToRun() -> (ReviewsToRun:Array of Object Soft Class References)
    __WorldContext:Object Reference=@n4
   ) -> (then, ObjectReviewMap:Map of Object References to Structures)
   > Display Review Results!(
    execute
    AssetMap:Map of Asset Data Structures to Structures
    ObjectMap:Map of Object References to Structures=@n5
    __WorldContext:Object Reference=@n4
   ) -> (then)
   n6 = RunObjectsThroughFilters!(
    execute=@n3.else
    Objects:Array of Object References=@n2
    ReportClasses:Array of GART Report Base Soft Class References=Get Report Classes() -> (Report Classes:Array of GART Report Base Soft Class References)
    __WorldContext:Object Reference=@n1
   ) -> (then, ReportDataMap:Map of Object References to Struct GART Asset Report List Structures)
   n7 = SpawnAndRegisterTabAndGetID!(
    execute=[@n6.then, @n9.then]
    self:Editor Utility Subsystem Object Reference=Get EditorUtilitySubsystem()
    InBlueprint:Editor Utility Widget Blueprint Object Reference=/Game/NeverCook/EditorUtilities/GenericAssetReviewTool/Widgets/EUW_GART_List_Main
   ) -> (then, NewTabID:Name=None, ReturnValue:Editor Utility Widget Object Reference)
   > n8 = Cast To EUW_GART_List_Main!(execute, Object=@n7.ReturnValue) -> (then, CastFailed, AsEUW GART List Main:EUW GART List Main Object Reference) {is=title K2Node_DynamicCast}
   > SetObjectReportData!(
    execute
    self:EUW GART List Main Object Reference=@n8
    ReportData:Map of Object References to Struct GART Asset Report List Structures=@n6
    UpdateList:bool=true
   ) -> (then)
   n9 = Branch(Condition="integer > integer"(A=Length!(TargetMap:Map of Object References to Struct GART Asset Report List Structures=@n6) -> (ReturnValue:int=0) {is=title K2Node_CallFunction}))
  ]}
 ReviewAllActors {nodes [
   ReviewAllActors() {is=title K2Node_FunctionEntry}
   > n1 = GetAllLevelActors(self=Get EditorActorSubsystem())
   > n2 = Branch(Condition=Is Control Down(KeysState=GetModifierKeysState()))
   n3 = Self-Reference()
   > n4 = Run Reviews on Objects(Objects=@n1, Reviews=Get ReviewsToRun(), __WorldContext=@n3)
   > Display Review Results(ObjectMap=@n4, __WorldContext=@n3)
   n5 = RunObjectsThroughFilters(execute=@n2.else, Objects=@n1, ReportClasses=Get Report Classes(), __WorldContext=Get l_world())
   > Branch(Condition="integer > integer"(A=Length(TargetMap=@n5) -> (ReturnValue=0)))
   > n6 = SpawnAndRegisterTabAndGetID(self=Get EditorUtilitySubsystem(), InBlueprint=&EUW_GART_List_Main) -> (NewTabID=None)
   > n7 = Cast To EUW_GART_List_Main(Object=@n6.ReturnValue)
   > SetObjectReportData(self=@n7, ReportData=@n5, UpdateList=true)
  ]}
]
variables [
 Report Classes {
  type="array<Class>"
  default [
   /Game/NeverCook/EditorUtilities/GenericAssetReviewTool/BP/ReportExamples/Actor/Report_ActorLightAttenuation_Base.Report_ActorLightAttenuation_Base_C
   /Game/NeverCook/EditorUtilities/GenericAssetReviewTool/BP/ReportExamples/Actor/Report_ActorLightFalloff_Base.Report_ActorLightFalloff_Base_C
   /Game/NeverCook/EditorUtilities/GenericAssetReviewTool/BP/ReportExamples/Actor/Report_BuildingAttachment_Base.Report_BuildingAttachment_Base_C
   /Game/NeverCook/EditorUtilities/GART/BR_Env/Reports/BREnv_Actor_CanBeDamaged.BREnv_Actor_CanBeDamaged_C
  ]
 }
 ReviewsToRun {
  type="array<Class>"
  default [
   /ContentReviewToolBREnv/Reviews/Actors/Review_Actor_PropAttachment.Review_Actor_PropAttachment_C
   /ContentReviewToolBREnv/Reviews/Actors/Review_Actor_CanBeDamaged.Review_Actor_CanBeDamaged_C
  ]
 }
]
